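# Sound cue for the global music stinger at the end of the GreenGlop Rift Bubble event, part of the Rift Beacon event sounds.
# A single wave player feeds the output directly with the end stinger wave, with no randomization, mixing or modulation.
# It uses a global celebration attenuation asset with a large maximum distance and routes through the medium-priority events sound class.
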SoundCue /Figment_S10/Athena/Sounds/Events/RiftBeacon/RiftBubble/GreenGlop/RiftBubble_GreenGlop_End_Music_Global_Cue
root=WavePlayer_0
nodes [{
  id=WavePlayer_0
  properties {soundWaveAssetPtr=/FNGameplayCues/Athena/Sounds/Events/RiftBeacon/RiftBubble/GreenGlop/Emote_GG_TT_MX_Stng_End_01}
 }]
properties {
 attenuationSettings=&/RiftBubble_GreenGlop_Global_Celebration_Att
 maxDistance=25000
 soundClassObject=/Figment_S10/Athena/Sounds/Events/SoundClasses/Fort_Events_Priority_Medium
}
