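# Sound Cue for the pre-start music stinger of the GreenGlop rift beacon bubble, in its "Global" variant.
# A single Wave Player plays one stinger wave from FNGameplayCues, with no randomization or mixing.
# It uses the GreenGlop global attenuation asset for long-range falloff and routes through the Fort_DanceMusic_Quest_PSM sound class.
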SoundCue /Figment_S10/Athena/Sounds/Events/RiftBeacon/RiftBubble/GreenGlop/RiftBubble_GreenGlop_Pre_Music_Global_Cue
root=WavePlayer_0
nodes [{
  id=WavePlayer_0
  properties {soundWaveAssetPtr=/FNGameplayCues/Athena/Sounds/Events/RiftBeacon/RiftBubble/GreenGlop/Emote_GG_TT_MX_Stng_PreStart_01}
 }]
properties {
 attenuationSettings=&/RiftBubble_GreenGlop_Global_Att
 maxDistance=20000
 soundClassObject=/Game/Sounds/FortSoundClassesAndModes/Fort_DanceMusic_Quest_PSM
 volumeMultiplier=1
}
