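# Sound cue for the global SFX of the GreenGlop Rift Bubble event, on a medium-priority events sound class.
# A mixer layers two branches, each with its own attenuation asset.
# A looping GG_TT emote SFX wave fades in through an enveloper volume curve and uses the celebration attenuation.
# A one-shot start stinger from the same emote set plays alongside it under the standard global attenuation.
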SoundCue /Figment_S10/Athena/Sounds/Events/RiftBeacon/RiftBubble/GreenGlop/RiftBubble_GreenGlop_SFX_Global_Cue
root=Mixer_0
nodes [
 {
  id=Attenuation_0
  children [Enveloper_0]
  properties {attenuationSettings=&/RiftBubble_GreenGlop_Global_Celebration_Att}
 }
 {
  id=Attenuation_1
  children [WavePlayer_2]
  properties {attenuationSettings=&/RiftBubble_GreenGlop_Global_Att}
 }
 {
  id=Enveloper_0
  children [WavePlayer_1]
  properties {pitchCurve {editorCurveData {keys [{time=0}]}}, volumeCurve {editorCurveData {keys [<at 0 0>,<at 3 1>]}}}
 }
 {id=Mixer_0, children [Attenuation_0, Attenuation_1], properties {inputVolume [1,1]}}
 {
  id=WavePlayer_1
  properties {bLooping=true, soundWaveAssetPtr=&/Emote_GG_TT_SFX_Loop_01}
 }
 {
  id=WavePlayer_2
  properties {soundWaveAssetPtr=&/Emote_GG_TT_SFX_Stng_Start_01}
 }
]
properties {maxDistance=25000, soundClassObject=/Figment_S10/Athena/Sounds/Events/SoundClasses/Fort_Events_Priority_Medium, volumeMultiplier=0.7}
